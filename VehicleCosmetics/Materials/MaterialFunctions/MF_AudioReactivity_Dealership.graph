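# Material function that supplies high and low frequency audio-reactivity signals for vehicle cosmetics, apparently as a dealership stand-in for live audio.
# The default-on FakeFrequency switch outputs fake pulses made by sampling two noise textures with looping, speed-scaled Time as UVs, then scaling by intensity parameters.
# With the switch off, outputs are red and green placeholder constants.
# The MF_AudioReactivity amplitude multiply feeds no output.
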MaterialFunction /VehicleCosmetics/Materials/MaterialFunctions/MF_AudioReactivity_Dealership
graph {nodes [
  n1 = StaticSwitchParameter(True=NamedRerouteUsage(), False=Constant3Vector() {desc=Placeholder for real reactive content, constant=<rgba 1 0 0 1>}) {defaultValue=true, desc=Enabling will output pulsing standalone frequencies., group=1_Audio Reactivity, parameterName=FakeFrequency}
  FunctionOutput(None=@n1) {outputName=HiFrequency_Out}
  n2 = StaticSwitchParameter(True=NamedRerouteUsage(), False=Constant3Vector() {desc=Placeholder for real reactive content, constant=<rgba 0 1 0 1>}) {defaultValue=true, desc=Enabling will output pulsing standalone frequencies., group=1_Audio Reactivity, parameterName=FakeFrequency}
  FunctionOutput(None=@n2) {outputName=LowFrequency_Out}
  Multiply(
   A=MaterialFunctionCall() {materialFunction=/Game/Packages/Fortress_SharedMaterials/Base_Material_Functions/BaseCharacterShaderFunctions/Audio/MF_AudioReactivity}.Amplitude
   B=Constant3Vector() {constant=<rgba 0.526 0.526 0.526 0>}
  )
  n3 = Multiply(A=Time() {bOverride_Period=true, period=8}, B=ScalarParameter() {defaultValue=1, group=1_Audio Reactivity, parameterName=LowFrequencySpeed})
  NamedRerouteDeclaration(None=Multiply(
    A=TextureSample(UVs=@n3) {samplerType=SAMPLERTYPE_LinearColor, texture=/Game/Global/Textures/Default/Tileable_Patterns/T_PerlinNoise_Uncompressed}.R
    B=ScalarParameter() {defaultValue=1, group=1_Audio Reactivity, parameterName=LowFrequencyIntensity}
   )) {name=FakeLowFrequencyOut, nodeColor=<rgba 1 0.5156 0 1>}
  n4 = Multiply(
   A=Time() {bOverride_Period=true, period=8}
   B=ScalarParameter() {defaultValue=0.5, group=1_Audio Reactivity, parameterName=HighFrequencySpeed}
  )
  NamedRerouteDeclaration(None=Multiply(
    A=TextureSample(UVs=@n4) {samplerType=SAMPLERTYPE_LinearColor, texture=/Game/Weapons/FORT_Melee/Pickaxe_RhombCamo/FX/T_Pickaxe_RhombCamo_Perlin}.R
    B=ScalarParameter() {defaultValue=1, group=1_Audio Reactivity, parameterName=HighFrequencyIntensity}
   )) {name=FakeHighFrequencyOut, nodeColor=<rgba 1 0.5156 0 1>}
 ]}
