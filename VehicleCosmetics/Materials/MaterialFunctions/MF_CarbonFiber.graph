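# Material function for vehicle cosmetics that generates a procedural carbon fiber weave and outputs it as material attributes.
# Tiled UVs feed a sine wave, offset per row by a floored V coordinate to stagger the strands.
# The absolute sine, darkened on alternate half-waves via a Ceil-driven lerp, forms the CarbonFiberHeight reroute.
# A scaled-down copy of that pattern drives both BaseColor and Roughness, with no texture samples.
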MaterialFunction /VehicleCosmetics/Materials/MaterialFunctions/MF_CarbonFiber
graph {nodes [
  n1 = Multiply(B=NamedRerouteUsage()) {constA=0.1}
  FunctionOutput(None=SetMaterialAttributes!(MaterialAttributes, BaseColor=@n1, Metallic, Roughness=@n1) {attributeSetTypes [#1, #2, #3]})
  n2 = Multiply(A=TextureCoordinate() {vTiling=2}, B=Constant() {r=32})
  n3 = Multiply(A=@n2) {constB=2}
  n4 = Multiply(A=ComponentMask(None=Floor(None=@n3)) {g=true}) {constB=0.125}
  n5 = Multiply(A=@n2) {constB=0.5}
  n6 = ComponentMask(None=Sine(None=Add(A=@n4, B=@n5)) {period=0.5}) {r=true}
  n7 = Abs(None=@n6)
  NamedRerouteDeclaration(None=LinearInterpolate(A=@n7, B=Multiply(A=@n7) {constB=0.1}, Alpha=Ceil(None=@n6))) {name=CarbonFiberHeight, nodeColor=<rgba 1 0 0.0937 1>}
 ]}
