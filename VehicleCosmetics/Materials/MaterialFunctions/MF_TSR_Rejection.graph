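# Material function for vehicle cosmetic materials that exposes a selectable TSR temporal rejection setting.
# An enumeration-controlled scalar parameter in the TSR_Settings group drives a Switch that outputs 0 for None or 1 for Full.
# Full rejection trades some aliasing for less smearing on fine or fast-moving animated materials, while None keeps TSR defaults.
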MaterialFunction /VehicleCosmetics/Materials/MaterialFunctions/MF_TSR_Rejection
graph {nodes [
  n1 = ScalarParameter() {
   controlType=Enumeration
   desc="None: No temporal dampening effects occur, TSR remains at default settings.\r\nFull: Full temporal rejction with the cost of some aliasing. Use for fine detail animation or quickly moving animated materials to prevent smearing.\r\n"
   enumeration=/VehicleCosmetics/Materials/Enums/Enum_TemporalRejection
   group=TSR_Settings
   parameterName=TSR_Rejection
  }
  MaterialExpressionConstant_2 = Constant()
  MaterialExpressionConstant_0 = Constant() {r=1}
  n2 = Switch!(SwitchValue=@n1, Default, TSR_Rejection_None=@MaterialExpressionConstant_2, TSR_Rejection_Full=@MaterialExpressionConstant_0) {inputs [
    {
     inputName=TSR_Rejection_None
     input {
      expression=&self:MaterialExpressionConstant_2
      inputName=None
      mask=0
      maskA=0
      maskB=0
      maskG=0
      maskR=0
      outputIndex=0
     }
    }
    {
     inputName=TSR_Rejection_Full
     input {
      expression=&self:MaterialExpressionConstant_0
      inputName=None
      mask=0
      maskA=0
      maskB=0
      maskG=0
      maskR=0
      outputIndex=0
     }
    }
   ]}
  FunctionOutput(None=@n2) {outputName=TSRRejection_Out}
 ]}
