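# MF_VCVFX_RainbowColor is a library-exposed material function that blends an input default color with an animated rainbow color for VFX materials.
# UVs, defaulting to stretched texture coordinates, pan along X at an input speed, and the fractional U value samples a rainbow curve in the thruster color curve atlas.
# A lerp input selects between default color and rainbow, with speed and blend inputs falling back to RainbowColor-group scalar parameters.
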MaterialFunction /DealershipSharedLibrary/FX/Materials/MaterialFunctions/MF_VCVFX_RainbowColor
graph {
 properties {bExposeToLibrary=true}
 nodes [
  n1 = FunctionInput() {bUsePreviewValueAsDefault=true, inputName=Default Color, sortPriority=1, previewValue=<1 1 1 1>}
  n2 = FunctionInput(Preview=TextureCoordinate() {uTiling=0.05, vTiling=0}) {bUsePreviewValueAsDefault=true, inputName=UVs, inputType=FunctionInput_Vector2, sortPriority=2, previewValue=<3 0 0 1>}
  n3 = FunctionInput(Preview=ScalarParameter() {defaultValue=3, group=RainbowColor, parameterName=RainbowPanerSpeedX}) {bUsePreviewValueAsDefault=true, inputName=Panner Speed Panner X, inputType=FunctionInput_Scalar, sortPriority=3, previewValue=<3 0 0 1>}
  n4 = Panner(Coordinate=@n2, Speed=AppendVector(A=@n3, B=Constant())) {speedX=3}
  n5 = CurveAtlasRowParameter(CurveTime=Frac(None=ComponentMask(None=@n4) {r=true})) {
   atlas=/DealershipSharedLibrary/FX/Textures/Curves/CA_VCVFX_ThrusterColor
   curve=/DealershipSharedLibrary/FX/Textures/Curves/Curve_VCVFX_BaseColorRainbow
   group=RainbowColor
   parameterName=CurveAtlas
  }
  n6 = ScalarParameter() {group=RainbowColor, parameterName=RainbowColor}
  FunctionOutput(None=LinearInterpolate(
    A=@n1
    B=@n5
    Alpha=FunctionInput(Preview=@n6) {bUsePreviewValueAsDefault=true, inputName="Rainbow Lerp 0=Default Color 1=Rainbow", inputType=FunctionInput_Scalar}
   ))
 ]
}
