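# Material function that layers carbon fiber onto incoming vehicle material attributes using a carbon fiber mask input.
# The static switch CarbonFiberToBodyShader selects either the shared MF_CarbonFiber_VC or MF_Vehicle_Body_Base as the layer, replacing carbon fiber with the body shader when on.
# The mask blends all attributes, then Normal is restored from the input and ClearCoatBottomNormal is explicitly lerped between input and layer.
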MaterialFunction /VehicleCosmetics/Materials/MaterialFunctions/MF_Apply_CarbonFiber
graph {nodes [
  n1 = GetMaterialAttributes(None=FunctionInput() {inputName=MaterialAttributes, inputType=FunctionInput_MaterialAttributes}) {attributeGetTypes [#1, #2]}
  n2 = StaticSwitchParameter(
   True=MaterialFunctionCall() {materialFunction=&/MF_Vehicle_Body_Base}.Result
   False=MaterialFunctionCall() {materialFunction=/VehicleCosmetics/SharedMaterials/MF_CarbonFiber_VC}.Result
  ) {
   desc=This switch will replace Carbon Fiber with the Body shader.
   group=SpecialFeatures
   parameterName=CarbonFiberToBodyShader
   sortPriority=34
  }
  n3 = GetMaterialAttributes(None=@n2) {attributeGetTypes [#2]}
  n4 = FunctionInput() {inputName=INCarbonFIberMask, sortPriority=1}
  FunctionOutput(None=SetMaterialAttributes!(
    MaterialAttributes=BlendMaterialAttributes(A=@n1.MaterialAttributes, B=@n3.MaterialAttributes, Alpha=@n4)
    Normal=@n1.Normal
    ClearCoatBottomNormal=LinearInterpolate(A=@n1.ClearCoatBottomNormal, B=@n3.ClearCoatBottomNormal, Alpha=@n4)
   ) {attributeSetTypes [#1, #2]}) {outputName=SetMaterialAttributes}
 ]}
