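# MF_BodyNormalLOD is a material function for vehicle cosmetics that layers a body normal map beneath a material's existing normal.
# The BodyNormalLOD texture, blank by default, is interpolated from a flat normal by the LOD scalar parameter.
# BlendAngleCorrectedNormals uses that result as the base and the incoming normal as the detail, then writes it back into the Material Attributes.
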MaterialFunction /VehicleCosmetics/Materials/MaterialFunctions/MF_BodyNormalLOD
graph {nodes [
  n1 = GetMaterialAttributes(None=FunctionInput() {inputType=FunctionInput_MaterialAttributes}) {attributeGetTypes [#1]}
  n2 = TextureSampleParameter2D() {group=Texture, parameterName=BodyNormalLOD, samplerType=SAMPLERTYPE_Normal, texture=/VehicleCosmetics/Textures/T_BlankNormal}
  n3 = LinearInterpolate(A=Constant3Vector() {constant=<rgba 0 0 1 0>}, B=@n2.RGB, Alpha=ScalarParameter() {parameterName=LOD}) {constAlpha=0}
  FunctionOutput(None=SetMaterialAttributes!(
    MaterialAttributes=@n1.MaterialAttributes
    Normal=MaterialFunctionCall(BaseNormal=@n3, AdditionalNormal=@n1.Normal) {materialFunction=&BlendAngleCorrectedNormals}.Result
   ) {attributeSetTypes [#1]})
 ]}
