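# MF_Fresnel_Prime is a material function in VehicleCosmetics that computes a tangent-space Fresnel term.
# It transforms the world camera vector into tangent space and dots it with the Normal input.
# The result is inverted with OneMinus and raised to the Fresnel Power input to shape the falloff.
# No clamp or absolute value is applied to the dot product before the power.
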MaterialFunction /VehicleCosmetics/Materials/MaterialFunctions/MF_Fresnel_Prime
graph {nodes [
  n1 = DotProduct(
   A=Transform(None=CameraVectorWS()) {transformSourceType=TRANSFORMSOURCE_World, transformType=TRANSFORM_Tangent}
   B=FunctionInput() {inputName=Normal, previewValue=<0 0 1 1>}
  )
  FunctionOutput(None=Power(Base=OneMinus(None=@n1), Exp=FunctionInput() {inputName=Fresnel Power, inputType=FunctionInput_Scalar, previewValue=<3 3 3 1>}) {constExponent=3})
 ]}
