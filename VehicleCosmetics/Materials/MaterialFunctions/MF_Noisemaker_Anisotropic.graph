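# Material function that generates animated, procedural anisotropic streak noise from UVs (channel 1 by default), with controls for tiling, angle, rotation, panning and chaos.
# Rotated, panned UVs split into columns, each shifted along Y by a chaos-driven step and warped by two counter-rotating gradients.
# Frac and triangle waves form the streaks and raw gradient; a half-offset second layer, with edge-fade masks hiding column seams, yields max and product variants.
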MaterialFunction /VehicleCosmetics/Materials/MaterialFunctions/MF_Noisemaker_Anisotropic
graph {nodes [
  n1 = FunctionInput(Preview=TextureCoordinate() {coordinateIndex=1}) {bUsePreviewValueAsDefault=true, inputName=Texture Coordinates, inputType=FunctionInput_Vector2, sortPriority=-1}
  n2 = FunctionInput(Preview=Constant()) {bUsePreviewValueAsDefault=true, inputName=Angle, inputType=FunctionInput_Scalar, sortPriority=5}
  n3 = Time() {bOverride_Period=true, period=6000}
  n4 = Constant()
  n5 = Rotator(
   Coordinate=Rotator(Coordinate=ConstantBiasScale(None=@n1) {bias=-0.5, scale=1}, Time=Divide(A=@n2) {constB=57.2958}) {centerX=0, centerY=0, speed=1}
   Time=Multiply(A=@n3, B=FunctionInput(Preview=@n4) {bUsePreviewValueAsDefault=true, inputName=RotateSpeed, inputType=FunctionInput_Scalar, sortPriority=4})
  ) {centerX=0, centerY=0, speed=1}
  n6 = FunctionInput(Preview=Constant()) {bUsePreviewValueAsDefault=true, inputName=PanSpeedX, inputType=FunctionInput_Scalar, sortPriority=2}
  n7 = Multiply(
   A=Add(A=ComponentMask(None=@n5) {r=true}, B=Multiply(A=@n3, B=@n6))
   B=FunctionInput(Preview=Constant() {r=50}) {bUsePreviewValueAsDefault=true, inputName=Tiling X, inputType=FunctionInput_Scalar}
  )
  n8 = Floor(None=@n7)
  n9 = FunctionInput(Preview=Constant() {r=1}) {bUsePreviewValueAsDefault=true, inputName=ChaosSpeed, inputType=FunctionInput_Scalar, sortPriority=7}
  n10 = Cosine(None=Multiply(A=Multiply(A=@n3) {constB=0.02}, B=@n9) {constB=0.02})
  n11 = FunctionInput(Preview=Constant() {r=1}) {bUsePreviewValueAsDefault=true, inputName=Chaos, inputType=FunctionInput_Scalar, sortPriority=6}
  n12 = LinearInterpolate(B=LinearInterpolate(Alpha=Abs(None=@n10)) {constA=0.45, constB=0.3}, Alpha=Saturate(None=@n11)) {constA=0.5, constB=0.3}
  n13 = FunctionInput(Preview=Constant() {r=1}) {bUsePreviewValueAsDefault=true, inputName=Tiling Y, inputType=FunctionInput_Scalar, sortPriority=1}
  n14 = Multiply(A=TextureCoordinate() {coordinateIndex=1}, B=Saturate(None=@n13))
  n15 = Multiply(A=@n3, B=@n9) {constB=0.02}
  n16 = ConstantBiasScale(None=Rotator(Coordinate=@n14, Time=@n15)) {bias=-0.5, scale=1}
  n17 = OneMinus(None=Abs(None=ComponentMask(None=@n16) {g=true}))
  n18 = ConstantBiasScale(None=Rotator(Coordinate=@n14, Time=@n15) {speed=-0.5}) {bias=-0.5, scale=1}
  n19 = OneMinus(None=Abs(None=ComponentMask(None=@n18) {g=true}))
  n20 = Abs(None=Multiply(A=ConstantBiasScale(None=@n17) {bias=-0.75, scale=2}, B=ConstantBiasScale(None=@n19) {bias=-0.75, scale=2}))
  n21 = Add(A=Multiply(A=ComponentMask(None=@n5) {g=true}, B=@n13), B=Multiply(A=@n11, B=ConstantBiasScale(None=@n20) {bias=-0.02, scale=1}))
  n22 = Multiply(
   A=@n3
   B=FunctionInput(Preview=Constant() {r=0.25}) {bUsePreviewValueAsDefault=true, inputName=PanSpeedY, inputType=FunctionInput_Scalar, sortPriority=3}
  )
  n23 = Add(A=Add(A=Multiply(A=@n8, B=@n12), B=@n21), B=@n22)
  n24 = Subtract(A=@n23, B=Floor(None=@n23))
  n25 = Abs(None=ConstantBiasScale(None=@n24) {bias=-0.5, scale=2})
  FunctionOutput(None=@n25) {outputName=Aniso1}
  FunctionOutput(None=@n24) {outputName=AnisoGradient, sortPriority=4}
  n26 = Add(A=@n7) {constB=0.5}
  n27 = Floor(None=@n26)
  n28 = Abs(None=ConstantBiasScale(None=Subtract(A=@n26, B=@n27)) {bias=-0.5, scale=2})
  n29 = Multiply(A=Saturate(None=Multiply(A=@n28) {constB=1.2}), B=@n25)
  FunctionOutput(None=@n29) {outputName=Aniso2, sortPriority=1}
  n30 = Multiply(A=@n12, B=@n27)
  n31 = Add(A=Add(A=@n22, B=Add(A=@n30, B=@n21))) {constB=0.25}
  n32 = Subtract(A=@n31, B=Floor(None=@n31))
  n33 = Floor(None=@n7)
  n34 = Abs(None=ConstantBiasScale(None=Subtract(A=@n7, B=@n33)) {bias=-0.5, scale=2})
  n35 = Multiply(A=Abs(None=ConstantBiasScale(None=@n32) {bias=-0.5, scale=2}), B=Saturate(None=Multiply(A=@n34) {constB=1.2}))
  FunctionOutput(None=Max(A=@n29, B=@n35)) {outputName=Aniso3, sortPriority=2}
  FunctionOutput(None=Multiply(A=@n29, B=@n35)) {outputName=Aniso4, sortPriority=3}
 ]}
