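# Material function that procedurally generates animated, stretched anisotropic noise without texture samples.
# Input UVs, defaulting to channel 1, are rotated by Angle and Rotation Speed, then feed four differently scaled layers with separate X/Y tiling.
# Each layer rotates and pans over time, then turns fractional coordinates into radial falloff dots sharpened by a power.
# Three outputs combine the layers as a sum of paired products, a plain sum and a full product.
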MaterialFunction /VehicleCosmetics/Materials/MaterialFunctions/MF_Noisemaker_AnisotropicSmooth
graph {nodes [
  n1 = AppendVector(
   A=FunctionInput(Preview=Constant() {r=50}) {bUsePreviewValueAsDefault=true, inputName=Tiling X, inputType=FunctionInput_Scalar}
   B=FunctionInput(Preview=Constant() {r=0.5}) {bUsePreviewValueAsDefault=true, inputName=Tiling Y, inputType=FunctionInput_Scalar, sortPriority=1}
  )
  n2 = Multiply(A=@n1) {constB=0.5}
  n3 = TextureCoordinate() {coordinateIndex=1}
  n4 = FunctionInput(Preview=Constant()) {bUsePreviewValueAsDefault=true, inputName=Angle, inputType=FunctionInput_Scalar, sortPriority=5}
  n5 = Constant()
  n6 = Time() {bOverride_Period=true, period=6000}
  n7 = Rotator(
   Coordinate=Rotator(
    Coordinate=FunctionInput(Preview=@n3) {bUsePreviewValueAsDefault=true, inputName=Texture Coordinates, inputType=FunctionInput_Vector2, sortPriority=-1}
    Time=Divide(A=@n4) {constB=57.2958, desc=Radian Remover}
   ) {speed=1}
   Time=Multiply(
    A=FunctionInput(Preview=@n5) {bUsePreviewValueAsDefault=true, inputName=Rotation Speed, inputType=FunctionInput_Scalar, sortPriority=5}
    B=@n6
   )
  ) {speed=1}
  n8 = Add(A=@n7)
  n9 = Multiply(A=@n2, B=ConstantBiasScale(None=@n8) {bias=-0.5, scale=2})
  n10 = Constant() {r=2}
  n11 = Multiply(
   A=@n6
   B=Multiply(A=FunctionInput(Preview=@n10) {bUsePreviewValueAsDefault=true, desc=Disorder, inputName=Chaos Speed, inputType=FunctionInput_Scalar, sortPriority=4}) {constB=0.01}
  )
  n12 = Multiply(
   A=@n6
   B=FunctionInput(Preview=Constant() {r=1}) {bUsePreviewValueAsDefault=true, inputName=PanSpeedY, inputType=FunctionInput_Scalar, sortPriority=3}
  )
  n13 = Multiply(
   A=@n6
   B=FunctionInput(Preview=Constant() {r=2}) {bUsePreviewValueAsDefault=true, inputName=PanSpeedX, inputType=FunctionInput_Scalar, sortPriority=2}
  )
  n14 = Panner(Coordinate=Panner(Coordinate=Rotator(Coordinate=@n9, Time=@n11) {speed=1}, Time=@n12) {speedY=0.7}, Time=@n13) {speedX=0.6}
  n15 = ConstantBiasScale(None=Subtract(A=@n14, B=Floor(None=@n14))) {bias=-0.5, scale=1.5}
  n16 = FunctionInput(Preview=Constant() {r=6}) {bUsePreviewValueAsDefault=true, inputName=Sharpen, inputType=FunctionInput_Scalar, sortPriority=6}
  n17 = Power(Base=OneMinus(None=DotProduct(A=@n15, B=@n15)), Exp=@n16)
  n18 = Multiply(A=@n2, B=ConstantBiasScale(None=@n8) {bias=-0.75, scale=3})
  n19 = Panner(Coordinate=Panner(Coordinate=Rotator(Coordinate=@n18, Time=@n11) {centerY=0.2, speed=0.4}, Time=@n12) {speedY=0.35}, Time=@n13) {speedX=0.275}
  n20 = ConstantBiasScale(None=Subtract(A=@n19, B=Floor(None=@n19))) {bias=-0.5, scale=1.45}
  n21 = Power(Base=OneMinus(None=DotProduct(A=@n20, B=@n20)), Exp=@n16)
  n22 = Multiply(A=@n17, B=@n21)
  n23 = Multiply(A=ConstantBiasScale(None=@n7) {bias=-1, scale=2}, B=@n1)
  n24 = Panner(Coordinate=Panner(Coordinate=Rotator(Coordinate=@n23, Time=@n11) {speed=1}, Time=@n12) {speedY=1}, Time=@n13) {speedX=1}
  n25 = ConstantBiasScale(None=Subtract(A=@n24, B=Floor(None=@n24))) {bias=-0.5, scale=1.5}
  n26 = Power(Base=OneMinus(None=DotProduct(A=@n25, B=@n25)), Exp=@n16)
  n27 = Multiply(A=@n1, B=ConstantBiasScale(None=@n7) {bias=-0.5, scale=1.5})
  n28 = Panner(Coordinate=Panner(Coordinate=Rotator(Coordinate=@n27, Time=@n11) {centerY=0.2, speed=0.4}, Time=@n12) {speedY=0.2}, Time=@n13) {speedX=0.2}
  n29 = ConstantBiasScale(None=Subtract(A=@n28, B=Floor(None=@n28))) {bias=-0.5, scale=1.45}
  n30 = Power(Base=OneMinus(None=DotProduct(A=@n29, B=@n29)), Exp=@n16)
  n31 = Multiply(A=@n26, B=@n30)
  FunctionOutput(None=Add(A=@n22, B=@n31)) {outputName=Aniso1}
  FunctionOutput(None=Add(A=Add(A=@n26, B=@n30), B=Add(A=@n17, B=@n21))) {outputName=Aniso2, sortPriority=1}
  FunctionOutput(None=Multiply(A=@n31, B=@n22)) {outputName=Aniso3, sortPriority=2}
 ]}
