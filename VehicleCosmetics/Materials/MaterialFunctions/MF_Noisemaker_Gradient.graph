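# Material function for vehicle cosmetics that generates animated, tiling linear gradient and stripe patterns from UV coordinates.
# Centred UVs are rotated by a degree angle and a time-driven spin, then the G channel is panned by wrapped time plus an offset and scaled by tiling.
# Ceil minus value forms a sawtooth, reshaped into three Sharpen-powered gradients and stripes, while the ceil index isolates a single-band mask.
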MaterialFunction /VehicleCosmetics/Materials/MaterialFunctions/MF_Noisemaker_Gradient
graph {nodes [
  n1 = FunctionInput(Preview=Constant() {r=10}) {bUsePreviewValueAsDefault=true, inputName=Tiling, inputType=FunctionInput_Scalar}
  n2 = ConstantBiasScale(None=FunctionInput(Preview=TextureCoordinate() {coordinateIndex=1}) {bUsePreviewValueAsDefault=true, inputName=Texture Coordinates, inputType=FunctionInput_Vector2, sortPriority=-1}) {bias=-0.5, scale=2}
  n3 = Divide(A=FunctionInput(Preview=Constant()) {bUsePreviewValueAsDefault=true, inputName=Angle, inputType=FunctionInput_Scalar, sortPriority=2}) {constB=57.2958, desc=Radian Remover}
  n4 = Time() {bOverride_Period=true, period=600}
  n5 = FunctionInput(Preview=Constant()) {bUsePreviewValueAsDefault=true, inputName=RotateSpeed, inputType=FunctionInput_Scalar, sortPriority=3}
  n6 = ComponentMask(None=Rotator(Coordinate=Rotator(Coordinate=@n2, Time=@n3) {centerX=0, centerY=0, speed=1}, Time=Multiply(A=@n4, B=@n5)) {centerX=0, centerY=0, speed=1}) {g=true}
  n7 = Multiply(B=FunctionInput(Preview=Constant()) {bUsePreviewValueAsDefault=true, inputName=PanSpeed, inputType=FunctionInput_Scalar, sortPriority=1}) {constA=0.1}
  n8 = FunctionInput(Preview=Constant()) {
   bUsePreviewValueAsDefault=true
   description=Keep betweeen approx 0-1 for best effect
   inputName=Offset
   inputType=FunctionInput_Scalar
   sortPriority=6
  }
  n9 = Add(A=Fmod(A=Multiply(A=@n4, B=@n7) {constA=0.1}, B=Constant() {r=2}), B=@n8)
  n10 = Multiply(A=@n1, B=ConstantBiasScale(None=Add(A=@n6, B=@n9)) {bias=0})
  n11 = Ceil(None=@n10)
  n12 = Subtract(A=@n11, B=@n10)
  n13 = FunctionInput(Preview=Constant() {r=1}) {bUsePreviewValueAsDefault=true, inputName=Sharpen, inputType=FunctionInput_Scalar, sortPriority=4}
  n14 = Power(Base=@n12, Exp=@n13)
  FunctionOutput(None=@n14) {outputName=Gradient1}
  n15 = ConstantBiasScale(None=@n12) {bias=-0.5, scale=2}
  n16 = Saturate(None=OneMinus(None=Abs(None=@n15)))
  FunctionOutput(None=Power(Base=@n16, Exp=@n13)) {outputName=Gradient2, sortPriority=1}
  FunctionOutput(None=Power(Base=ConstantBiasScale(None=@n16) {bias=-0.5, scale=2}, Exp=@n13)) {outputName=Gradient3, sortPriority=2}
  FunctionOutput(None=Saturate(None=ConstantBiasScale(None=@n14) {bias=-0.5, scale=2})) {outputName=Stripes, sortPriority=3}
  n17 = Abs(None=Subtract(A=Subtract(A=@n11, B=@n1)))
  FunctionOutput(None=Saturate(None=OneMinus(None=@n17))) {outputName=OneLineMask, sortPriority=4}
 ]}
