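# Material function that generates animated, Perlin-style procedural noise without texture samples, under the vehicle cosmetics materials.
# Tiled, angled UVs feed four layers with differing scales, Chaos rotation offsets, spin and pan rates.
# Each layer builds repeating soft radial dots from fractional UVs and a dot-product falloff, sharpened by Balance.
# The summed layers output as Perlin1, a darkened Perlin2, and a folded, darkened Perlin3 using absolute one-minus.
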MaterialFunction /VehicleCosmetics/Materials/MaterialFunctions/MF_Noisemaker_Perlin
graph {nodes [
  n1 = FunctionInput(Preview=Constant() {r=2}) {bUsePreviewValueAsDefault=true, inputName=Tiling X, inputType=FunctionInput_Scalar}
  n2 = FunctionInput(Preview=Constant() {r=1}) {bUsePreviewValueAsDefault=true, inputName=Tiling Y, inputType=FunctionInput_Scalar, sortPriority=1}
  n3 = FunctionInput(Preview=TextureCoordinate() {coordinateIndex=1}) {bUsePreviewValueAsDefault=true, inputName=Texture Coordinates, inputType=FunctionInput_Vector2, sortPriority=-1}
  n4 = Constant()
  n5 = Rotator(
   Coordinate=Multiply(A=AppendVector(A=@n1, B=@n2), B=ConstantBiasScale(None=@n3) {bias=-0.5, scale=2})
   Time=Divide(A=FunctionInput(Preview=@n4) {bUsePreviewValueAsDefault=true, inputName=Angle, inputType=FunctionInput_Scalar, sortPriority=4}) {constB=57.2958, desc=Radian remover}
  ) {centerX=0, centerY=0, speed=1}
  n6 = FunctionInput(Preview=Constant() {r=0.25}) {bUsePreviewValueAsDefault=true, desc=directional randomization, inputName=Chaos, inputType=FunctionInput_Scalar, sortPriority=6}
  n7 = Rotator(Coordinate=@n5, Time=@n6) {centerX=0, centerY=0, speed=1}
  n8 = Time() {bOverride_Period=true, period=6000}
  n9 = Multiply(
   A=@n8
   B=FunctionInput(Preview=Constant()) {bUsePreviewValueAsDefault=true, inputName=RotateSpeed, inputType=FunctionInput_Scalar, sortPriority=4}
  )
  n10 = Multiply(
   A=@n8
   B=FunctionInput(Preview=Constant() {r=0.2}) {bUsePreviewValueAsDefault=true, inputName=PanSpeedY, inputType=FunctionInput_Scalar, sortPriority=3}
  )
  n11 = Multiply(
   A=@n8
   B=FunctionInput(Preview=Constant() {r=0.2}) {bUsePreviewValueAsDefault=true, inputName=PanSpeedX, inputType=FunctionInput_Scalar, sortPriority=2}
  )
  n12 = Panner(Coordinate=Panner(Coordinate=Rotator(Coordinate=@n7, Time=@n9) {centerX=0, centerY=0, speed=1}, Time=@n10) {speedY=1}, Time=@n11) {speedX=1}
  n13 = ConstantBiasScale(None=Subtract(A=@n12, B=Floor(None=@n12))) {bias=-0.5, scale=1.9}
  n14 = Saturate(None=OneMinus(None=DotProduct(A=@n13, B=@n13)))
  n15 = FunctionInput(Preview=Constant() {r=2.75}) {bUsePreviewValueAsDefault=true, inputName=Balance, inputType=FunctionInput_Scalar, sortPriority=5}
  n16 = Rotator(Coordinate=@n5, Time=@n6) {speed=-1}
  n17 = Multiply(A=@n16, B=Constant2Vector() {g=1.25, r=1.5})
  n18 = Panner(Coordinate=Panner(Coordinate=Rotator(Coordinate=@n17, Time=@n9) {centerX=0, centerY=0, speed=-1}, Time=@n10) {speedY=0.5}, Time=@n11) {speedX=0.75}
  n19 = ConstantBiasScale(None=Subtract(A=@n18, B=Floor(None=@n18))) {bias=-0.5, scale=1.9}
  n20 = Saturate(None=OneMinus(None=DotProduct(A=@n19, B=@n19)))
  n21 = Multiply(A=@n7, B=Constant2Vector() {g=1.5, r=2})
  n22 = Panner(Coordinate=Panner(Coordinate=Rotator(Coordinate=@n21, Time=@n9) {centerX=0, centerY=0, speed=0.5}, Time=@n10) {speedY=0.25}, Time=@n11) {speedX=0.25}
  n23 = ConstantBiasScale(None=Subtract(A=@n22, B=Floor(None=@n22))) {bias=-0.5, scale=1.9}
  n24 = Saturate(None=OneMinus(None=DotProduct(A=@n23, B=@n23)))
  n25 = Multiply(A=@n16, B=Constant2Vector() {g=1.75, r=0.95})
  n26 = Panner(Coordinate=Panner(Coordinate=Rotator(Coordinate=@n25, Time=@n9) {centerX=0, centerY=0, speed=0.4}, Time=@n10) {speedY=1.1}, Time=@n11) {speedX=1.1}
  n27 = ConstantBiasScale(None=Subtract(A=@n26, B=Floor(None=@n26))) {bias=-0.5, scale=1.9}
  n28 = Saturate(None=OneMinus(None=DotProduct(A=@n27, B=@n27)))
  n29 = Add(A=Add(A=Power(Base=@n14, Exp=@n15), B=Power(Base=@n20, Exp=@n15)), B=Add(A=Power(Base=@n24, Exp=@n15), B=Power(Base=@n28, Exp=@n15)))
  FunctionOutput(None=@n29) {outputName=Perlin1}
  FunctionOutput(None=ConstantBiasScale(None=@n29) {bias=-0.25, scale=1}) {outputName=Perlin2, sortPriority=1}
  n30 = OneMinus(None=@n29)
  FunctionOutput(None=ConstantBiasScale(None=Abs(None=@n30)) {bias=-0.25, scale=1}) {outputName=Perlin3, sortPriority=2}
 ]}
