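# Material function that builds a base plastic material layer for vehicle cosmetics from Plastic parameters for color, metallic, roughness, clear coat and clear coat roughness.
# Attributes are set with the Clear Coat shading model, then pass through the BlendShadingModel composite to the output.
# The composite blends to a default shading model copy using Saturate(Ceil(ClearCoat)) as a hard switch, so Clear Coat applies only above zero.
# An angle-corrected normal blend is present but does not feed the output.
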MaterialFunction /VehicleCosmetics/Materials/MaterialFunctions/MF_Plastic_Base
graph {nodes [
  n1 = ScalarParameter() {group=Plastic, parameterName=Plastic_ClearCoat}
  n2 = SetMaterialAttributes!(
   MaterialAttributes
   BaseColor=VectorParameter() {group=Plastic, parameterName=Plastic_Color, defaultValue=<rgba 0.03 0.03 0.03 1>}.RGB
   Metallic=ScalarParameter() {group=Plastic, parameterName=Plastic_Metallic}
   Roughness=ScalarParameter() {defaultValue=0.65, group=Plastic, parameterName=Plastic_Roughness}
   CustomData0=@n1
   CustomData1=ScalarParameter() {group=Plastic, parameterName=Plastic_ClearCoatRoughness}
   ShadingModel=ShadingModel() {shadingModel=MSM_ClearCoat}
  ) {attributeSetTypes [
    #1
    #2
    #3
    #4
    #5
    #6
   ]}
  Composite!("In (MA)"=@n2, CCThickneess=@n1) {subgraphName=BlendShadingModel}
  MaterialExpressionReroute_61 = Reroute(None=@n2)
  MaterialExpressionReroute_0 = Reroute(None=@n1)
  n3 = BlendMaterialAttributes(
   A=SetMaterialAttributes!(MaterialAttributes=@MaterialExpressionReroute_61."In (MA)", ShadingModel=ShadingModel()) {attributeSetTypes [#6]}
   B=@MaterialExpressionReroute_61."In (MA)"
   Alpha=Saturate(None=Ceil(None=@MaterialExpressionReroute_0.CCThickneess))
  )
  MaterialExpressionReroute_63 = Reroute(None=@n3)
  FunctionOutput(None=@MaterialExpressionReroute_63.Output)
  n4 = GetMaterialAttributes(None=@MaterialExpressionReroute_61."In (MA)") {attributeGetTypes [#7, #8]}
  MaterialFunctionCall(BaseNormal=@n4.Normal, AdditionalNormal=@n4.Normal) {materialFunction=&BlendAngleCorrectedNormals}
  PinBase() {reroutePins [
    "In (MA)" {expression=&self:MaterialExpressionReroute_61}
    CCThickneess {expression=&self:MaterialExpressionReroute_0}
   ]}
  PinBase!(Output=@n3) {reroutePins [Output {expression=&self:MaterialExpressionReroute_63}]}
 ]}
