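# Material function that builds the base trim layer for vehicle cosmetic materials, outputting material attributes from tintable color, metallic, roughness and clear coat parameters.
# An ApplyTrim static switch sources AO, curvature and emissive mask from ChassisTextures and Masks, or from packed BodyTextures channels when off.
# AO multiplies the trim color.
# Optional emissive, gated by ApplyTrimEmissive, multiplies a remapped, saturated mask by color, brightness and curvature via a named reroute.
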MaterialFunction /VehicleCosmetics/Materials/MaterialFunctions/MF_Trim_Base
graph {nodes [
  n1 = TextureSampleParameter2D() {
   group=Textures
   parameterName=ChassisTextures
   samplerSource=SSM_Wrap_WorldGroupSettings
   samplerType=SAMPLERTYPE_LinearColor
   sortPriority=35
   texture=/VehicleCosmetics/Textures/T_Black_Linear
   channelNames {a=Roughness, b=Metallic, g=AO-BlackOut, r=Curvature}
  }
  n2 = TextureSampleParameter2D() {
   group=Texture
   parameterName=BodyTextures
   samplerSource=SSM_Wrap_WorldGroupSettings
   samplerType=SAMPLERTYPE_LinearColor
   texture=&T_Black_Linear
  }
  n3 = TextureSampleParameter2D() {
   desc=_M
   group=Textures
   parameterName=Masks
   samplerSource=SSM_Wrap_WorldGroupSettings
   samplerType=SAMPLERTYPE_LinearColor
   sortPriority=33
   texture=&T_Black_Linear
   channelNames {a="", b=BoostMask, g=PaintMask, r=EmssivePartsMask}
  }
  n4 = Subtract(A=StaticSwitchParameter(True=@n3.EmssivePartsMask, False=@n2.B) {parameterName=ApplyTrim}) {constB=0.098}
  n5 = Multiply(
   A=Saturate(None=Multiply(A=@n4) {constB=1.1087})
   B=Multiply(
    A=VectorParameter() {group=Trim, parameterName=Trim_EmissiveColor, sortPriority=50, defaultValue=<rgba 0 1 1 1>}.RGB
    B=ScalarParameter() {defaultValue=1, group=Trim, parameterName=Trim_EmissiveBrightness, sortPriority=51}
   )
  )
  n6 = NamedRerouteUsage()
  n7 = SetMaterialAttributes!(
   MaterialAttributes
   BaseColor=Multiply(
    A=VectorParameter() {group=Trim, parameterName=TrimColor, defaultValue=<rgba 0.03 0.03 0.03 1>}.RGB
    B=StaticSwitchParameter(True=@n1.AO-BlackOut, False=@n2.G) {parameterName=ApplyTrim}
   )
   Metallic=ScalarParameter() {group=Trim, parameterName=Trim_Metallic}
   Roughness=ScalarParameter() {defaultValue=0.65, group=Trim, parameterName=Trim_Roughness}
   CustomData0=ScalarParameter() {group=Trim, parameterName=Trim_ClearCoat}
   CustomData1=ScalarParameter() {group=Trim, parameterName=Trim_ClearCoatRoughness}
   EmissiveColor=StaticSwitchParameter(True=Multiply(A=@n5, B=@n6), False=Constant()) {group=Trim, parameterName=ApplyTrimEmissive}
  ) {attributeSetTypes [
    #1
    #2
    #3
    #4
    #5
    #6
   ]}
  FunctionOutput(None=@n7)
  NamedRerouteDeclaration(None=StaticSwitchParameter(True=@n1.Curvature, False=@n2.R) {parameterName=ApplyTrim}) {name=Curvature, nodeColor=<rgba 0 0.3672 1 1>}
 ]}
