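# Material function that builds PBR material attributes for vehicle trim livery, including clear coat and its bottom normal, output as TrimLivery_MA.
# Channel-packed textures supply curvature, AO-blackout, metallic and roughness, with a static ApplyTrim switch taking AO from either the chassis or body texture.
# Base color combines AO with a curvature overlay, and a blackout mask adjusts color and roughness.
# A flattened trim normal feeds CCBottomNormal; its LOD-weighted blend with the body normal is left unconnected.
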MaterialFunction /VehicleCosmetics/Materials/MaterialFunctions/MF_Trim_PBR
graph {nodes [
  n1 = SetMaterialAttributes!(
   MaterialAttributes
   BaseColor=NamedRerouteUsage()
   Metallic=LinearInterpolate(A=NamedRerouteUsage(), Alpha=NamedRerouteUsage())
   Roughness=NamedRerouteUsage()
   AmbientOcclusion=NamedRerouteUsage()
   CustomData0=ScalarParameter() {group=Trim, parameterName=Trim_ClearCoat}
   CustomData1=ScalarParameter() {group=Trim, parameterName=Trim_ClearCoatRoughness}
   ClearCoatBottomNormal=NamedRerouteUsage()
  ) {attributeSetTypes [
    #1
    #2
    #3
    #4
    #5
    #6
    #7
   ]}
  FunctionOutput(None=@n1) {description="Carries BaseColor, Metallic, Rough, Normal and AO", outputName=TrimLivery_MA, sortPriority=-10}
  n2 = TextureSampleParameter2D() {group=Texture, parameterName=BodyNormalLOD, samplerType=SAMPLERTYPE_Normal, texture=/VehicleCosmetics/Textures/T_BlankNormal}
  n3 = TextureSampleParameter2D() {group=Finish, parameterName=TrimBottomNormalBase, samplerType=SAMPLERTYPE_Normal, texture=/VehicleCosmetics/Textures/T_BlankNormal_B}
  n4 = MaterialFunctionCall(Normal=@n3.RGB, Flatness=ScalarParameter() {group=DetailNormal, parameterName=TrimBottomNormalFlatness}) {materialFunction=&FlattenNormal}
  MaterialFunctionCall(
   BaseNormal=LinearInterpolate(A=Constant3Vector() {constant=<rgba 0 0 1 0>}, B=@n2.RGB, Alpha=ScalarParameter() {parameterName=LOD}) {constAlpha=0}
   AdditionalNormal=@n4.Result
  ) {materialFunction=&BlendAngleCorrectedNormals}
  NamedRerouteDeclaration(None=@n4.Result) {name=CCBottomNormal, nodeColor=<rgba 1 0.6563 0 1>}
  n5 = TextureSampleParameter2D() {
   group=Textures
   parameterName=Normal
   samplerSource=SSM_Wrap_WorldGroupSettings
   samplerType=SAMPLERTYPE_Normal
   sortPriority=34
   texture=&T_BlankNormal_B
  }
  NamedRerouteDeclaration(None=@n5.RGB) {name=NormalTexture, nodeColor=<rgba 0.6038 0.3325 0.013 1>}
  n6 = TextureSampleParameter2D() {
   group=Textures
   parameterName=BaseColor
   samplerSource=SSM_Wrap_WorldGroupSettings
   sortPriority=30
   texture=/VehicleCosmetics/Textures/T_Black
  }
  NamedRerouteDeclaration(None=@n6.RGB) {name=BaseColorTexture, nodeColor=<rgba 0.6038 0.3325 0.013 1>}
  n7 = Add(A=NamedRerouteUsage(), B=LinearInterpolate(A=Constant() {r=-1}, B=Constant() {r=1}, Alpha=Constant() {r=0.55}))
  n8 = NamedRerouteUsage()
  NamedRerouteDeclaration(None=Clamp(None=LinearInterpolate(A=@n7, Alpha=@n8) {desc=BlackOut})) {name=Rough_out, nodeColor=<rgba 0.2422 1 0 1>}
  n9 = NamedRerouteUsage()
  n10 = Multiply(A=Clamp(None=NamedRerouteUsage()), B=NamedRerouteUsage()) {desc=AO}
  n11 = NamedRerouteUsage()
  NamedRerouteDeclaration(None=Multiply(
    A=OneMinus(None=@n9)
    B=MaterialFunctionCall(Base=@n10, Blend=@n11) {desc=curvature, materialFunction=&Blend_Overlay}.Result
   ) {desc=BlackOut}) {name=BaseColorFinal, nodeColor=<rgba 1 0.6328 0 1>}
  n12 = TextureSampleParameter2D() {
   group=Textures
   parameterName=ChassisTextures
   samplerSource=SSM_Wrap_WorldGroupSettings
   samplerType=SAMPLERTYPE_LinearColor
   sortPriority=35
   texture=/VehicleCosmetics/Textures/T_Black_Linear
   channelNames {a=Roughness, b=Metallic, g=AO-BlackOut, r=Curvature}
  }
  n13 = TextureSampleParameter2D() {
   group=Trim
   parameterName=BodyTextures
   samplerSource=SSM_Wrap_WorldGroupSettings
   samplerType=SAMPLERTYPE_LinearColor
   sortPriority=35
   texture=&T_Black_Linear
   channelNames {a=Roughness, b=Metallic, g=AO-BlackOut, r=Curvature}
  }
  NamedRerouteDeclaration(None=StaticSwitchParameter(True=@n12.AO-BlackOut, False=@n13.AO-BlackOut) {group=Trim, parameterName=ApplyTrim}) {name=AO, nodeColor=<rgba 0 0.1726 0.4792 1>}
  n14 = Ceil(None=NamedRerouteUsage())
  NamedRerouteDeclaration(None=OneMinus(None=Clamp(None=@n14))) {name=Blackout, nodeColor=<rgba 0.5703 1 0 1>}
  NamedRerouteDeclaration(None=@n13.Curvature) {name=Curvature, nodeColor=<rgba 0 0.1726 0.4792 1>}
  NamedRerouteDeclaration(None=@n13.Roughness) {name=Roughness, nodeColor=<rgba 0 0.1726 0.4792 1>}
  NamedRerouteDeclaration(None=@n13.Metallic) {name=Metallic, nodeColor=<rgba 0 0.1726 0.4792 1>}
 ]}
