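# MF_Trim is a material function for vehicle cosmetic trim that outputs material attributes, selecting between a base and a PBR trim variant.
# It blends the results of MF_Trim_Base and MF_Trim_PBR by the UsePBRSwitch scalar parameter.
# A UsePBR_BodyTextures static switch takes the same inputs but does not feed the output.
# Both parameters' descriptions require the PBR texture to have "Compress Without Alpha" disabled.
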MaterialFunction /VehicleCosmetics/Materials/MaterialFunctions/MF_Trim
graph {nodes [
  n1 = MaterialFunctionCall() {materialFunction=&/MF_Trim_Base}
  n2 = MaterialFunctionCall() {materialFunction=&/MF_Trim_PBR}
  n3 = ScalarParameter() {
   desc="When using, ensure PBR texture has \"Compress Without Alpha\" set to False. "
   group=Trim
   parameterName=UsePBRSwitch
   sortPriority=-10
  }
  FunctionOutput(None=BlendMaterialAttributes(A=@n1.Result, B=@n2.TrimLivery_MA, Alpha=@n3))
  StaticSwitchParameter(True=@n2.TrimLivery_MA, False=@n1.Result) {
   desc="When using, ensure PBR texture has \"Compress Without Alpha\" set to False. "
   group=Trim
   parameterName=UsePBR_BodyTextures
   sortPriority=-10
  }
 ]}
