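# Material function that applies shared post effects to vehicle cosmetic materials, taking material attributes in and returning them modified.
# The input passes through MF_BaseVehicle_Damage, then its original shading model is written back onto the damaged result.
# That result feeds MF_VehicleDeRez_Transition, whose output becomes the function's output.
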MaterialFunction /VehicleCosmetics/Materials/MaterialFunctions/MF_Vehicle_FNPostFX
graph {nodes [
  n1 = GetMaterialAttributes(None=FunctionInput() {inputType=FunctionInput_MaterialAttributes}) {attributeGetTypes [#1]}
  n2 = MaterialFunctionCall(Material=@n1.MaterialAttributes) {materialFunction=&MF_BaseVehicle_Damage}
  FunctionOutput(None=MaterialFunctionCall(In=SetMaterialAttributes!(MaterialAttributes=@n2.Result, ShadingModel=@n1.ShadingModel) {attributeSetTypes [#1]}) {materialFunction=&MF_VehicleDeRez_Transition}.Result)
 ]}
